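# Material for the CoveSpice pistol, named for its blank hull and equip animation.
# It uses Masked blend with Default Lit shading and is flagged for skeletal meshes and mesh deformers, with automatic usage setting off in the editor.
# The graph holds a single Constant node with no connection shown, so the file shows no textures, parameters or animation logic.
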Material /WildEstateWeapons/Assets/CoveSpice_Pistol/Materials/MI_CoveSpice_Pistol_BlankHull_EquipAnim
graph {
 properties {
  bAutomaticallySetUsageInEditor=false
  bUsedWithMeshDeformer=true
  bUsedWithSkeletalMesh=true
  blendMode=BLEND_Masked
  usedShadingModels=Default Lit
 }
 nodes [Constant()]
}
